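; Cinematic paint launcher weapon Blueprint that removes itself from the player's inventory once it runs out of ammo.
; On each projectile spawn it runs the parent logic, then checks the remaining ammo.
; When none is left, a timer set with a configurable removal delay fires a custom event.
; That event casts the instigator's controller to a Fort player controller and removes the item by its inventory GUID.
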(event FortWeapon|EventOnSpawnProjectile
  (params (SpawnProjectile :type Fort%20Projectile%20Base%20Object%20Reference))
  (|CallParentFunction SpawnProjectile)
  ;[1 1 1] Remove the weapon when it runs out of ammo
  (if (<= (FortWeapon|GetRemainingAmmo) 0)
    (Utilities|Time|SetTimerbyEvent (delegate Custom|RemoveWeaponOnEmpty) (Variables|Default|GetRemovalDelay))))

(event Custom|RemoveWeaponOnEmpty
  ;[1 1 1] Remove the weapon when it runs out of ammo
  (bind _asfort_player_controller (Utilities|Casting|CastToFortPlayerController (Game|GetInstigatorController)))
  (Inventory|RemoveItemFromPlayerByGuid _asfort_player_controller (FortWeapon|GetInventoryGUID) -1 true))

(fn ConstructionScript
  (|CallParentFunction))
