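# Material function in the Papaya prototype weapon folder that outputs position, normal and bounds-relative values for materials.
# Calculated Bounds divides interpolated pre-skinned position minus bounds Min by Max minus Max, a zero divisor, rather than the bounds size.
# World Normals passes the pre-skinned normal through a vertex interpolator without a world transform.
# WorldPosition outputs world position, and LocalPosition transforms it from world to local space.
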MaterialFunction /Game/Athena/Items/Weapons/Prototype/Papaya/MaterialFunctions/MF_Papaya_CalculatedBounds
graph {nodes [
  n1 = VertexInterpolator(VS=PreSkinnedPosition())
  n2 = PreSkinnedLocalBounds()
  FunctionOutput(None=Divide(A=Subtract(A=@n1.PS, B=@n2.Min), B=Subtract(A=@n2.Max, B=@n2.Max))) {outputName=Calculated Bounds, sortPriority=2}
  FunctionOutput(None=VertexInterpolator(VS=PreSkinnedNormal()).PS) {outputName=World Normals, sortPriority=3}
  n3 = WorldPosition()
  FunctionOutput(None=@n3.XYZ) {outputName=WorldPosition}
  FunctionOutput(None=TransformPosition(None=@n3.XYZ) {transformSourceType=TRANSFORMPOSSOURCE_World}) {outputName=LocalPosition, sortPriority=1}
 ]}
