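# Translucent lit deferred decal for the Papaya prototype weapon that pairs a streaming television movie texture with paint-splat shapes.
# Splat mask and normal are picked from a subdivided atlas by index; decal lifetime drives mask erosion and normal strength.
# Screen derivatives give a surface-facing fade, bypassed on mobile.
# The movie uses plain or triplanar bounds-based UVs via static switch; UV-checker branches hold screen-space, centered and triplanar mapping experiments.
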Material /Game/Athena/Items/Weapons/Prototype/Papaya/M_Decal_Cinema
graph {
 properties {blendMode=BLEND_Translucent, materialDomain=MD_DeferredDecal, pixelDepthOffsetMode=PDOM_Legacy, usedShadingModels=Default Lit}
 nodes [
  ConstantBiasScale(None=Sine(None=Time()) {period=5})
  MaterialFunctionCall(TextureObject=TextureObject() {texture=/Game/Weapons/WeaponSkins/Wraps/Textures/UVCheaker}) {materialFunction=/Game/Weapons/WeaponSkins/Wraps/MaterialFunctions/WorldAlignedTexture}
  n1 = TextureCoordinate() {desc=Defaults to texture coord}
  n2 = TextureSample(UVs=@n1) {texture=/Game/UI/Foundation/Movie/StreamingTelevisionMediaTexture}
  Multiply(A=@n2.RGB, B=ScalarParameter() {defaultValue=0.25, parameterName=Emissive Intensity})
  n3 = VertexNormalWS()
  n4 = MaterialFunctionCall(Float3=Abs(None=@n3)) {materialFunction=&BreakOutFloat3Components}
  Multiply(A=@n4.R)
  n5 = AppendVector(A=ScalarParameter() {defaultValue=4, parameterName=SubdivisionsU}, B=ScalarParameter() {defaultValue=1, parameterName=SubdivisionsV})
  n6 = ScalarParameter() {parameterName=IndexU}
  n7 = ScalarParameter() {parameterName=IndexV}
  n8 = Add(A=Divide(A=TextureCoordinate(), B=@n5), B=Divide(A=AppendVector(A=@n6, B=@n7), B=@n5))
  n9 = TextureSampleParameter2D(UVs=@n8) {parameterName=Splat Mask, texture=&/T_PaintSplats_02_D}
  n10 = OneMinus(None=DecalLifetimeOpacity().Opacity)
  n11 = Add(A=@n9.G, B=SphereMask(A=@n9.B, B=Constant(), Radius=@n10, Hardness=ScalarParameter() {defaultValue=0.2, parameterName=ErosionHardness}))
  n12 = Multiply(A=@n10) {constB=2}
  n13 = Multiply(A=LinearInterpolate(B=OneMinus(None=@n9.G)) {constA=1, constAlpha=0.75}, B=Subtract(A=@n9.R, B=Multiply(A=@n11, B=@n12)))
  n14 = WorldPosition()
  n15 = DDX(Value=@n14.XYZ)
  n16 = DDY(Value=@n14.XYZ)
  n17 = DotProduct(A=Multiply(A=ObjectOrientation(), B=Constant() {r=-1}), B=CrossProduct(A=Normalize(VectorInput=@n15), B=Normalize(VectorInput=@n16)))
  n18 = Constant() {r=1}
  Multiply(A=Saturate(None=ShadingPathSwitch(Default=@n17, Mobile=@n18)), B=@n13)
  Reroute(None=@n2.RGB)
  n19 = TextureSampleParameter2D(UVs=@n8) {parameterName=Splat Mask Normal, samplerType=SAMPLERTYPE_Normal, texture=&/T_PaintSplats_02_N}
  n20 = VectorParameter() {parameterName=NormalOuterStength, defaultValue=<rgba 0.1 0.1 1 0>}
  Reroute(None=LinearInterpolate(A=@n19.RGB, B=Multiply(A=@n19.RGB, B=@n20.RGB), Alpha=@n10))
  Reroute(None=@n13) {desc=Disbled Edge Fade}
  ScalarParameter() {defaultValue=0.5, parameterName=Metallic}
  ScalarParameter() {defaultValue=0.7, parameterName=Specular}
  ScalarParameter() {defaultValue=0.25, parameterName=Roughness}
  n21 = TextureObject() {texture=&StreamingTelevisionMediaTexture}
  n22 = ScalarParameter() {defaultValue=1, parameterName=Cinema_TileAmount, sortPriority=0}
  n23 = MaterialFunctionCall() {materialFunction=/Game/Athena/Items/Weapons/Prototype/Papaya/MaterialFunctions/MF_Papaya_CalculatedBounds}
  n24 = Divide(A=@n23.WorldPosition, B=ScalarParameter() {defaultValue=1, parameterName=Cinema_WorldScale, sortPriority=0})
  n25 = ScalarParameter() {defaultValue=100, parameterName=Cinema_TriplanarContrast, sortPriority=0}
  StaticSwitchParameter(
   True=ComponentMask(None=MaterialFunctionCall!(TextureObject=@n21, TextureSize=@n22, WorldPosition=@n24, World Space Normal=@n23.World Normals, ProjectionTransitionContrast=@n25) {materialFunction=/Game/Athena/Items/Weapons/Prototype/Papaya/MaterialFunctions/MF_Papaya_TriplanarUVSimple}.XYZ Texture) {b=true, g=true, r=true}
   False=@n2.RGB
  ) {parameterName=UseTriplanarUVs, sortPriority=-1}
  n26 = WorldPosition()
  n27 = StaticBool()
  StaticSwitch(True=TransformPosition(None=@n26.XYZ) {transformSourceType=TRANSFORMPOSSOURCE_World}, False=@n26.XYZ, Value=@n27)
  StaticSwitch(
   True=Normalize(VectorInput=Transform(None=@n3) {transformSourceType=TRANSFORMSOURCE_World, transformType=TRANSFORM_Local})
   False=@n3
   Value=@n27
  )
  n28 = ConstantBiasScale(None=ScreenPosition().ViewportUV) {bias=-0.5, scale=2}
  n29 = MaterialFunctionCall(Float2=ViewSize()) {materialFunction=&BreakOutFloat2Components}
  n30 = AppendVector(A=Constant() {r=1}, B=Divide(A=@n29.G, B=@n29.G))
  n31 = Multiply(
   A=Constant2Vector() {g=-0.5, r=0.5}
   B=MaterialFunctionCall(World Position=ActorPositionWS()) {materialFunction=&TransformToClipSpace}.Clip Space XY
  )
  n32 = Subtract(A=ConstantBiasScale(None=Multiply(A=@n28, B=@n30)), B=Multiply(A=Multiply(A=@n30, B=@n31)) {constB=0.5})
  n33 = Max(A=@n4.R, B=@n4.R)
  n34 = ComponentMask(None=@n26.XYZ) {desc=YX, g=true, r=true}
  n35 = If(
   A=@n4.R
   B=@n4.R
   "A > B"=ComponentMask(None=@n26.XYZ) {b=true, desc=YZ, g=true}
   "A < B"=ComponentMask(None=@n26.XYZ) {b=true, desc=XZ, r=true}
  )
  n36 = StaticSwitchParameter(
   True=Divide(
    A=If(A=@n4.B, B=@n33, "A > B"=@n34, "A < B"=@n35)
    B=ScalarParameter() {defaultValue=250, parameterName=Triplanar_TileAmount, sortPriority=0}
   )
   False=@n1
  ) {parameterName=UseTriplanarUVs, sortPriority=-1}
  TextureSample(UVs=StaticSwitchParameter(True=ScreenPosition().ViewportUV, False=StaticSwitchParameter(True=@n32, False=@n36) {parameterName=UseScreenCenteredUVs, sortPriority=1}) {parameterName=UseScreenSpaceUVs, sortPriority=2}) {texture=&UVCheaker}
 ]
}
