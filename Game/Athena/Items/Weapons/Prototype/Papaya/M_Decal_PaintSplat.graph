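# Translucent deferred decal material for paint splats, tinting a two-color blend by the decal color.
# An atlas cell, chosen by index parameters or particle color red, drives the mask and normal textures.
# As decal lifetime opacity drops, packed mask channels and a SphereMask erode the splat and its normal shifts toward scaled strength.
# Screen-derivative surface normals mask faces misaligned with the projection, except on the mobile path.
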Material /Game/Athena/Items/Weapons/Prototype/Papaya/M_Decal_PaintSplat
graph {
 properties {blendMode=BLEND_Translucent, materialDomain=MD_DeferredDecal, pixelDepthOffsetMode=PDOM_Legacy, usedShadingModels=Default Lit}
 nodes [
  ConstantBiasScale(None=Sine(None=Time()) {period=5})
  n1 = AppendVector(A=ScalarParameter() {defaultValue=4, parameterName=SubdivisionsU}, B=ScalarParameter() {defaultValue=1, parameterName=SubdivisionsV})
  n2 = LinearInterpolate(A=ScalarParameter() {parameterName=IndexU}, B=ParticleColor().R, Alpha=ScalarParameter() {parameterName=UseParticleColorIndex})
  n3 = ScalarParameter() {parameterName=IndexV}
  n4 = Add(A=Divide(A=TextureCoordinate(), B=@n1), B=Divide(A=AppendVector(A=@n2, B=@n3), B=@n1))
  n5 = TextureSampleParameter2D(UVs=@n4) {parameterName=NormalMap, samplerType=SAMPLERTYPE_Normal, texture=&/T_PaintSplats_Normal}
  n6 = OneMinus(None=DecalLifetimeOpacity().Opacity)
  LinearInterpolate(A=@n5.RGB, B=Multiply(A=@n5.RGB, B=VectorParameter() {parameterName=NormalOuterStength, defaultValue=<rgba 0.1 0.1 1 0>}.RGB), Alpha=@n6)
  n7 = TextureSampleParameter2D(UVs=@n4) {parameterName=Texture, samplerType=SAMPLERTYPE_LinearColor, texture=&/T_PaintSplats_01}
  n8 = Add(A=@n7.G, B=SphereMask(A=@n7.B, B=Constant(), Radius=@n6, Hardness=ScalarParameter() {defaultValue=0.2, parameterName=ErosionHardness}))
  n9 = Multiply(A=@n6) {constB=2}
  n10 = WorldPosition()
  n11 = DDX(Value=@n10.XYZ)
  n12 = DDY(Value=@n10.XYZ)
  n13 = DotProduct(A=Multiply(A=ObjectOrientation(), B=Constant() {r=-1}), B=CrossProduct(A=Normalize(VectorInput=@n11), B=Normalize(VectorInput=@n12)))
  n14 = Constant() {r=1}
  Multiply(A=Subtract(A=@n7.R, B=Multiply(A=@n8, B=@n9)), B=Saturate(None=ShadingPathSwitch(Default=@n13, Mobile=@n14)))
  n15 = DecalColor()
  n16 = VectorParameter() {parameterName=DecalColorA, defaultValue=<rgba 0.6111 0.1125 0.0127 0.3>}
  n17 = LinearInterpolate(A=@n16.RGB, B=VectorParameter() {parameterName=DecalColorB, defaultValue=<rgba 0.5781 0.1169 0.0015 0.1>}.RGB, Alpha=@n7.G)
  Multiply(A=@n15.RGB, B=@n17)
  Multiply(A=@n15.RGB, B=Multiply(A=@n17, B=@n16.A))
  ScalarParameter() {defaultValue=0.5, parameterName=Metallic}
  ScalarParameter() {defaultValue=0.7, parameterName=Specular}
  ScalarParameter() {defaultValue=0.25, parameterName=Roughness}
 ]
}
