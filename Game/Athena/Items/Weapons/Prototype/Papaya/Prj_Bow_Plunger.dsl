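; Plunger bow projectile that sticks to what it hits, then explodes after a set delay, or immediately on "NoStick" components.
; HandleHits orients it to the surface and attaches it to pawns, or through HandleSticks to wall doors, vehicles and supply drops or their balloons.
; Hit handling and the unstick effect and sound on destruction are skipped when the instigating player is DBNO.
; A water-entry event spawns a depth-aware splash effect and sound.
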(event EventBeginPlay
  ;[1 1 1] Begin Play
  (Water|AssignOnEnterWaterEvent (Variables|Water|GetWaterInteractionComponent) (EventDispatchers|CreateEvent :!event Custom|OnEnterWaterEvent_Event_0))
  (bind _asfort_player_pawn (Utilities|Casting|CastToFortPlayerPawn (Variables|Actor|GetInstigator)))
  (Variables|Default|SetPlayerPawn _asfort_player_pawn))

(event Collision|EventHit
  (params (MyComp :type Primitive%20Component%20Object%20Reference) (Other :type Actor%20Object%20Reference) (OtherComp :type Primitive%20Component%20Object%20Reference) (bSelfMoved :type Boolean) (HitLocation :type Vector :default "0, 0, 0") (HitNormal :type Vector :default "0, 0, 0") (NormalImpulse :type Vector :default "0, 0, 0") (Hit :type Hit%20Result%20Structure%20%28by%20ref%29))
  (bind (_bblockinghit _binitialoverlap _time _distance _location _impactpoint _normal _impactnormal _physmat _hitactor _hitcomponent) (Collision|BreakHitResult Hit))
  (bind _lifespan_on_hit (Variables|Default|GetLifespan%20on%20Hit))
  ;[1 1 1] Stop event.
  (Variables|Default|SetImpactNormal HitNormal)
  (if (Utilities|NotEqual(Object) self _hitactor)
    (bind _playerpawn (Variables|Default|GetPlayerPawn)
      (:then
        (if (not (Variables|/Script/FortniteGame.FortPawn|GetbIsDBNO _playerpawn))
          (CallFunction|HandleHits Hit)
          (if (Components|ComponentHasTag _hitcomponent "NoStick")
            (Delivery|DoExplosion) ; If we hit a NoStick component pop right away
            (Actor|SetLifeSpan 0.01)
            (else
              (Actor|SetLifeSpan _lifespan_on_hit)
              (Utilities|FlowControl|Delay _lifespan_on_hit)
              (Delivery|DoExplosion)))))
      (:else))))

(event EventDestroyed
  ;[1 1 1] Destroyed
  (bind _playerpawn (Variables|Default|GetPlayerPawn)
    (:then
      (if (Variables|/Script/FortniteGame.FortPawn|GetbIsDBNO _playerpawn)
        _
        (elif (Variables|Default|GetHookStopped)
          (Effects|SpawnEmitteratLocation "/Game/Weapons/FORT_Crossbows/Crossbow/Effects/P_Hook_Impact_Break_Papaya.P_Hook_Impact_Break_Papaya" (Transformation|GetActorLocation) (Math|Conversions|RotationFromXVector (Variables|Default|GetImpactNormal)))
          (Audio|PlaySoundatLocation (Variables|Default|GetUnstickSound) (Transformation|GetActorLocation)))))
    (:else)))

(event Custom|OnEnterWaterEvent_Event_0
  (params (WaterBody :type Fort%20Water%20Body%20Actor%20Object%20Reference) (WaterInteractionComponent :type Fort%20Water%20Interaction%20Component%20Object%20Reference) (bIsFirstBody :type Boolean))
  (bind _waterinteractioncomponent (Variables|Water|GetWaterInteractionComponent))
  (bind (_waterplanelocation _waterplanenormal _watersurfaceposition _waterdepth _waterbodyidx) (FX|GetLastWaterSurfaceInfo _waterinteractioncomponent :!class /Script/FortniteGame.FortWaterInteractionComponent))
  ;[1 1 1] Water splash
  (if bIsFirstBody
    (bind _returnvalue (Effects|SpawnEmitteratLocation "/Game/Weapons/FORT_Generic_Effects/Impacts/Water/P_Pistol_Impact_Water_Apollo.P_Pistol_Impact_Water_Apollo" _watersurfaceposition (Math|Rotator|MakeRotfromX _waterplanenormal)))
    (Effects|Components|ParticleSystem|SetFloatParameter _returnvalue "WaterDepth" _waterdepth)
    (Effects|Components|ParticleSystem|SetFloatParameter _returnvalue "WaterBodyIndex" (Math|Conversions|ToFloat(Integer) _waterbodyidx))
    (Audio|SpawnSoundatLocation "/Game/Sounds/Fort_GamePlay_Sounds/Explosives/Grenade_Projectile_Bounce_Water_Splash_NoLoop_Cue.Grenade_Projectile_Bounce_Water_Splash_NoLoop_Cue" _waterplanelocation)))

(fn HandleHits
  (params (Hit :type Hit%20Result%20Structure))
  (bind (_bblockinghit _binitialoverlap _time _distance _location _impactpoint _normal _impactnormal _physmat _hitactor _hitcomponent) (Collision|BreakHitResult Hit))
  (Variables|Default|SetHookStopped true)
  (bind _output_get (Variables|Default|SetHitLocation _impactpoint))
  (Audio|PlaySoundatLocation (Variables|Default|GetStickSound) _output_get)
  (bind _output_get_2 (Variables|Default|SetImpactNormal _impactnormal))
  (Transformation|SetActorRotation (Math|Conversions|RotationFromXVector (* _output_get_2 -1.0)))
  (bind _asfort_player_pawn (Utilities|Casting|CastToFortPlayerPawn _hitactor)
    (:then
      (Gameplay|ServerAttachtoActor Hit _asfort_player_pawn (Variables|Default|GetOffset%20Distance%20from%20Physics%20Mesh) (Variables|Default|GetOffsetDistanceFromBone) 0.75))
    (:CastFailed
      (HandleSticks|HandleSticks _hitactor _hitcomponent)))
  (Variables|Default|SetStruckActor _hitactor)
  (return 0))

(fn HandleSticks
  (params (Object :type Object%20Reference) (Component :type Primitive%20Component%20Object%20Reference))
  (Utilities|FlowControl|Sequence
    (:then_0
      (HandleSticks|HandleStuckBuilding Object Component))
    (:then_1
      (HandleSticks|HandleStuckVehicle Object Component))
    (:then_2
      (HandleSticks|HandleStuckSupplyDrop Object Component))))

(fn HandleStuckBuilding
  (params (Object :type Object%20Reference) (Component :type Primitive%20Component%20Object%20Reference))
  (bind _asbuilding_wall (Utilities|Casting|CastToBuildingWall Object))
  (LocalVariables|SetLocalWall _asbuilding_wall)
  (bind _asstatic_mesh_component (Utilities|Casting|CastToStaticMeshComponent Component))
  (bind _returnvalue (BuildingWall|IsDoorComponent (LocalVariables|GetLocalWall) _asstatic_mesh_component))
  (if _returnvalue
    (Transformation|AttachActorToComponent Component "None" "KeepWorld" "KeepWorld" "KeepWorld" false)))

(fn HandleStuckProjectiles
  (params (Object :type Object%20Reference) (Component :type Primitive%20Component%20Object%20Reference))
  (Utilities|FlowControl|Sequence
    (:then_0
      (HandleSticks|HandleStuckBuilding Object Component))
    (:then_1
      (HandleSticks|HandleStuckVehicle Object Component))
    (:then_2
      (HandleSticks|HandleStuckSupplyDrop Object Component))))

(fn HandleStuckSupplyDrop
  (params (Object :type Object%20Reference) (Component :type Primitive%20Component%20Object%20Reference))
  (Utilities|FlowControl|Sequence
    (:then_0
      (Utilities|Casting|CastToAthenaSupplyDrop Object)
      (Transformation|AttachActorToComponent Component "None" "KeepWorld" "KeepWorld" "KeepWorld"))
    (:then_1
      (Utilities|Casting|CastToAthenaSupplyDropBalloon Object)
      (Transformation|AttachActorToComponent Component "None" "KeepWorld" "KeepWorld" "KeepWorld" false))))

(fn HandleStuckVehicle
  (params (Object :type Object%20Reference) (Component :type Primitive%20Component%20Object%20Reference))
  (Utilities|Casting|CastToFortAthenaVehicle Object)
  (Transformation|AttachActorToComponent Component "None" "KeepWorld" "KeepWorld" "KeepWorld" false))

(fn ConstructionScript
  (|CallParentFunction))
