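; Blue paint grenade projectile Blueprint from the Papaya prototype weapons, a child class that tints its particle trail.
; On BeginPlay it runs the parent logic, then, if the trail particle system is valid, converts the trail color variable to a vector and sets it as the "Color" parameter.
; The explosion event and construction script only call the parent, so any paint or impact behavior is not defined here.
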(event EventBeginPlay
  (|CallParentFunction)
  (bind _trailps (Variables|Default|GetTrailPS)
    (:then
      (Effects|Components|ParticleSystem|SetVectorParameter _trailps "Color" (Math|Conversions|ToVector(LinearColor) (Variables|Default|GetTrailColor))))
    (:else)))

(event EventOnExploded
  (params (HitActors :type Array%20of%20Actor%20Object%20References) (HitResults :type Array%20of%20Hit%20Result%20Structures))
  (|CallParentFunction HitActors HitResults))

(fn ConstructionScript
  (|CallParentFunction))
