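; Papaya prototype grenade projectile that explodes on a fuse timer, or on the authority when it hits something or enters water.
; On stop it records surface, normal and hit actor type, then spawns a water-landing effect or an explosion with a randomized decal, depth set by actor type.
; Out-of-water midair explosions add an airburst and three axis-aligned decals.
; Explosion cosmetics also include sound, a sound indicator and a feedback GameplayCue.
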(event EventBeginPlay
  ;[1 1 1] SpawnRibbonTrails
  (bind _returnvalue (Effects|SpawnEmitterAttached (Variables|Effects|GetRibbonTrail_PS) (Variables|Projectile_GrenadeLauncher|GetMesh)))
  (Components|Activation|Activate _returnvalue false)
  (Variables|Default|SetTrailPS _returnvalue)
  (Effects|Components|ParticleSystem|SetFloatParameter _returnvalue "TrailLifetime" (Variables|Effects|GetTrailLifetime))
  (Utilities|Time|SetTimerbyFunctionName self "FuseTimerMax" (Variables|Default|GetFuseTimer)))

(event EventOnStop
  (params (Hit :type Hit%20Result%20Structure%20%28by%20ref%29))
  (bind (_bblockinghit _binitialoverlap _time _distance _location _impactpoint _normal _impactnormal _physmat _hitactor) (Collision|BreakHitResult Hit))
  ;[1 1 1] StopBehavior, get PhysMat
  (|CallParentFunction Hit)
  (Variables|Default|SetStopLocation _impactpoint) ; Use Loc and Rot for single explosion
  (Variables|Default|SetNormal _normal)
  (Variables|Default|SetStopRotX (Math|Rotator|MakeRotfromX _normal))
  (Variables|Default|SetStopRotZ (Math|Rotator|MakeRotfromZ _normal))
  (Variables|Default|SetStopRotY (Math|Rotator|MakeRotfromY _normal))
  (Variables|Effects|SetStopSurfaceType (Variables|/Script/PhysicsCore.PhysicalMaterial|GetSurfaceType _physmat))
  (CallFunction|SetDecalActorType _hitactor)
  (Variables|Default|SetHasStopped true)
  (if (Utilities|Enum|Equal(Enum) (Variables|Effects|GetStopSurfaceType) "SurfaceType8")
    (Networking|CanExecuteCosmeticEvents
      (:True
        (Rendering|SetVisibility (Variables|Projectile_GrenadeLauncher|GetMesh) false true)
        (bind _returnvalue (Effects|SpawnEmitteratLocation (Variables|Effects|GetWaterLandDelayPS) (Variables|Default|GetStopLocation) (Variables|Default|GetStopRotX) "1.000000,1.000000,1.000000" true "AutoRelease"))
        (Variables|Default|SetWaterLandDelayRipplePSC _returnvalue)
        (Audio|PlaySoundatLocation "/Game/Sounds/Fort_GamePlay_Sounds/Explosives/Grenade_Projectile_Bounce_Water_Splash_NoLoop_Cue.Grenade_Projectile_Bounce_Water_Splash_NoLoop_Cue" (Transformation|GetActorLocation))))
    (else
      (Utilities|FlowControl|DoOnce false
        (:Completed
          (bind _returnvalue_2 (Effects|SpawnEmitteratLocation (Variables|Effects|GetExplosion_PS) (Variables|Default|GetStopLocation) (Variables|Default|GetStopRotZ) "1.000000,1.000000,1.000000" true "AutoRelease"))
          (Variables|Default|SetImpactPS _returnvalue_2)
          (CallFunction|SpawnDecal "NewEnumerator0"))))))

(event Custom|FuseTimerMax
  ;[1 1 1] Explode on Fuse Timer
  (if (not (Delivery|IsBeingKilled))
    (Delivery|DoExplosion)))

(event OnEnterWaterEvent(WaterInteractionComponent)
  (params (WaterBody :type Fort%20Water%20Body%20Actor%20Object%20Reference) (WaterInteractionComponent :type Fort%20Water%20Interaction%20Component%20Object%20Reference) (bIsFirstBody :type Boolean))
  ;[1 1 1] Explode on hit
  (Networking|SwitchHasAuthority
    (:Authority
      (Delivery|DoExplosion))))

(event Collision|EventHit
  (params (MyComp :type Primitive%20Component%20Object%20Reference) (Other :type Actor%20Object%20Reference) (OtherComp :type Primitive%20Component%20Object%20Reference) (bSelfMoved :type Boolean) (HitLocation :type Vector :default "0, 0, 0") (HitNormal :type Vector :default "0, 0, 0") (NormalImpulse :type Vector :default "0, 0, 0") (Hit :type Hit%20Result%20Structure%20%28by%20ref%29))
  ;[1 1 1] Explode on hit
  (Networking|SwitchHasAuthority
    (:Authority
      (Delivery|DoExplosion))))

(event EventOnExploded
  (params (HitActors :type Array%20of%20Actor%20Object%20References) (HitResults :type Array%20of%20Hit%20Result%20Structures))
  (bind _returnvalue (Transformation|GetActorLocation))
  (Networking|CanExecuteCosmeticEvents
    (:True
      ;[1 1 1] WorldObjectSurfaceType, Distinguish between water and everything else
      (Fortnite|ShowSoundIndicatorLocal :Location (Transformation|GetActorLocation) :MaxAudibleDistance 5e+03 :IndicatorType "Gunshot" :Instigator self :Tint "(R=0.000000,G=0.000000,B=0.000000,A=0.000000)" :AttenuationMaxRangeMultiplier 1.0 :AttenuationIntensityMultiplier 1.0)
      (if (not (Water|IsinWater (Variables|Water|GetWaterInteractionComponent)))
        (switch Utilities|FlowControl|Switch|SwitchonEPhysicalSurface (Variables|Effects|GetStopSurfaceType)
          (:SurfaceType_Default
            _)
          (:SurfaceType1
            _)
          (:SurfaceType2
            _)
          (:SurfaceType3
            _)
          (:SurfaceType4
            _)
          (:SurfaceType5
            _)
          (:SurfaceType6
            _)
          (:SurfaceType7
            _)
          (:SurfaceType9
            _)
          (:SurfaceType10
            _)
          (:SurfaceType11
            _)
          (:SurfaceType12
            _)
          (:SurfaceType13
            _)
          (:SurfaceType14
            _)
          (:SurfaceType15
            _)
          (:SurfaceType16
            _)
          (:SurfaceType17
            _)
          (:SurfaceType23
            _)
          (:SurfaceType24
            _))
        (if (not (Variables|Default|GetHasStopped))
          (Effects|SpawnEmitteratLocation (Variables|Effects|GetExplosion_InAir_PS) _returnvalue (Transformation|GetActorRotation) "1.000000,1.000000,1.000000" true "AutoRelease")
          (Variables|Default|SetStopLocation _returnvalue)
          (CallFunction|SpawnTriPlanarDecal))
        (Audio|PlaySoundatLocation (Variables|Sounds|GetGrenade_Explosion_Sound) (Transformation|GetActorLocation))
        (Rendering|SetVisibility (Variables|Projectile_GrenadeLauncher|GetMesh) false true)
        (GameplayCue|ExecuteGameplayCue%28Burst%29Local self (Variables|Default|GetFeedbackCue) (Ability|GameplayCue|MakeGameplayCueParameters 0.0)) ; Cam shake + controller vibration
      ))))

(fn GetDecalMaterial
  (return (Utilities|Array|Get(acopy) (Variables|Default|GetDecalMaterials) (Math|Random|RandomIntegerinRange 0 (Utilities|Array|LastIndex (Variables|Default|GetDecalMaterials))))))

(fn SetDecalActorType
  (params (HitActor :type Object%20Reference))
  (Utilities|Casting|CastToPlayerPawn_Athena HitActor
    (:then
      (Variables|Default|SetDecalActorType "NewEnumerator1")
      (return))
    (:CastFailed
      (Utilities|Casting|CastToLandscapeStreamingProxy (Variables|GetHitActor)
        (:then
          (Variables|Default|SetDecalActorType "NewEnumerator2")
          (return))
        (:CastFailed
          (Variables|Default|SetDecalActorType "NewEnumerator0")
          (return))))))

(fn SpawnBounceFX
  (params (Hit :type Hit%20Result%20Structure%20%28by%20ref%29))
  (bind (_bblockinghit _binitialoverlap _time _distance _location _impactpoint _normal _impactnormal _physmat _hitactor) (Collision|BreakHitResult Hit))
  (bind _returnvalue (Math|Conversions|RotationFromXVector _impactnormal))
  (bind _returnvalue_2 (Collision|GetFortPhysicalSurface Hit))
  (switch Utilities|FlowControl|Switch|SwitchonEPhysicalSurface _returnvalue_2
    (:SurfaceType_Default
      _)
    (:SurfaceType1
      _)
    (:SurfaceType2
      _)
    (:SurfaceType3
      _)
    (:SurfaceType4
      _)
    (:SurfaceType5
      _)
    (:SurfaceType6
      _)
    (:SurfaceType7
      _)
    (:SurfaceType8
      (Effects|SpawnEmitteratLocation (Variables|Effects|GetBounce_WaterSplash_PS) _impactpoint _returnvalue)
      (Audio|PlaySoundatLocation "/Game/Sounds/Fort_GamePlay_Sounds/Explosives/Snowball_Projectile_Bounce_Cue.Snowball_Projectile_Bounce_Cue" _impactpoint)
      (return _hitactor _returnvalue_2 _returnvalue))
    (:SurfaceType9
      _)
    (:SurfaceType10
      _)
    (:SurfaceType11
      _)
    (:SurfaceType12
      _)
    (:SurfaceType13
      _)
    (:SurfaceType14
      _)
    (:SurfaceType15
      _)
    (:SurfaceType16
      _)
    (:SurfaceType17
      _))
  (Effects|SpawnEmitteratLocation (Variables|Effects|GetBounce_Generic_PS) _impactpoint _returnvalue)
  (Audio|PlaySoundatLocation "/Game/Sounds/Fort_GamePlay_Sounds/Explosives/Snowball_Projectile_Bounce_Cue.Snowball_Projectile_Bounce_Cue" _impactpoint))

(fn SpawnDecal
  (params (CanSpawn :type En_Grenade_Papaya_ActorTypes%20Enum :default "NewEnumerator0"))
  (bind _output (CallFunction|GetDecalMaterial))
  (if (or (Utilities|Enum|Equal(Enum) CanSpawn "NewEnumerator1") (Math|Vector|Equal(Vector) (Variables|Default|GetStopLocation) "0, 0, 0" 0.0001))
    (Development|PrintString "DECAL FAILED" true true "(R=1.000000,G=0.000000,B=0.000000,A=1.000000)" 5.0)
    (else
      (Utilities|IsValid _output
        (:Is%20Valid
          (bind _returnvalue (Rendering|Decal|SpawnDecalatLocation _output (Math|Vector|MakeVector (Utilities|Select :Index (Variables|Default|GetDecalActorType) :NewEnumerator0 (Variables|Default|GetDecalDepth) :NewEnumerator1 0.0 :NewEnumerator2 (Variables|Default|GetDecalDepthLandscape)) (Math|Random|RandomFloatinRange (Variables|Default|GetDecalSizeMin) (Variables|Default|GetDecalSizeMax)) (Math|Random|RandomFloatinRange (Variables|Default|GetDecalSizeMin) (Variables|Default|GetDecalSizeMax))) (Variables|Default|GetStopLocation) (Math|Rotator|MakeRotfromX (* (Variables|Default|GetNormal) -1.0)) (Variables|Default|GetDecalLifetime)))
          (Transformation|AddLocalRotation _returnvalue (Math|Rotator|MakeRotator (Math|Random|RandomFloatinRange 0.0 3.6e+02)) false false)
          (Rendering|Components|Decal|SetFadeIn _returnvalue 0.0 (Variables|Default|GetDecalFadeInDuration))
          (Rendering|Components|Decal|SetFadeOut _returnvalue (Variables|Default|GetDecalDelayBeforeFade) (Variables|Default|GetDecalFadeOutDuration) true))))))

(fn SpawnTriPlanarDecal
  (Variables|Default|SetNormal "1.000000, 0,0.000000")
  (CallFunction|SpawnDecal "NewEnumerator0")
  (Variables|Default|SetNormal "0.000000,1.000000,0.000000")
  (CallFunction|SpawnDecal "NewEnumerator0")
  (Variables|Default|SetNormal "0.000000,0.000000,1.000000")
  (CallFunction|SpawnDecal "NewEnumerator0"))

(fn ConstructionScript
  (Effects|Components|ParticleSystem|SetTemplate (Variables|Default|GetRibbonTrail_PSC) (Variables|Effects|GetRibbonTrail_PS))
  (Effects|Components|ParticleSystem|SetTemplate (Variables|Default|GetWaterSplash_PSC) (Variables|Effects|GetBounce_WaterSplash_PS))
  (Effects|Components|ParticleSystem|SetTemplate (Variables|Default|GetBounce_Generic_PSC) (Variables|Effects|GetBounce_Generic_PS))
  (Effects|Components|ParticleSystem|SetTemplate (Variables|Default|GetExplosion_PSC) (Variables|Effects|GetExplosion_PS))
  (Effects|Components|ParticleSystem|SetTemplate (Variables|Default|GetExplosion_Water_PSC) 0)
  (AddComponent|NODE_AddParticleSystemComponent-1))
